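# Translucent unlit Niagara sprite material for flame effects on the CritterBran character.
# It builds a shaped gradient from UVs, depth-fades it by noise, then erodes it with two blended, panning noise layers.
# The eroded value indexes a FireGradient texture, tinted by particle color, with Fresnel-scaled alpha.
# Per-particle random time offsets drive animation, while Low and Medium quality skip UV waving and depth fade.
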Material /BRCosmetics/Characters/Player/Male/Medium/Bodies/M_MED_CritterBran/FX/Materials/M_Character_CritterBran_Flame
graph {
 properties {
  bAutomaticallySetUsageInEditor=false
  bUsedWithNiagaraSprites=true
  blendMode=BLEND_Translucent
  shadingModel=MSM_Unlit
  usedShadingModels=Unlit
 }
 nodes [
  n1 = ParticleColor()
  n2 = TextureCoordinate()
  n3 = OneMinus(None=ComponentMask(None=@n2) {g=true})
  n4 = ScalarParameter() {defaultValue=0.45, parameterName=Length Fade Gradient Exponent}
  n5 = ConstantBiasScale(None=OneMinus(None=Power(Base=@n3, Exp=@n4))) {bias=-0.5, scale=2}
  n6 = DotProduct(A=@n5, B=@n5)
  n7 = ConstantBiasScale(None=ComponentMask(None=@n2) {r=true}) {bias=-0.5, scale=2}
  n8 = DotProduct(A=@n7, B=@n7)
  n9 = Multiply(
   A=Power(Base=OneMinus(None=@n6), Exp=ScalarParameter() {defaultValue=0.5, parameterName=Length Exponent})
   B=Power(Base=OneMinus(None=@n8), Exp=ScalarParameter() {defaultValue=1.666, parameterName=Width Exponent})
  )
  n10 = TextureCoordinate()
  n11 = OneMinus(None=ComponentMask(None=@n10) {g=true})
  n12 = ScalarParameter() {defaultValue=0.666, parameterName=UV Exponent V}
  n13 = AppendVector(A=ComponentMask(None=@n10) {r=true}, B=OneMinus(None=Power(Base=@n11, Exp=@n12)))
  n14 = VectorParameter() {parameterName=UV Wave Parameters, defaultValue=<rgba 1 0.62 0.05 0.1>}
  n15 = ScalarParameter() {defaultValue=1.62, parameterName=Wave Speed}
  n16 = NamedRerouteUsage()
  n17 = TextureCoordinate()
  n18 = QualitySwitch(
   Default=NamedRerouteDeclaration(None=MaterialFunctionCall!(
     UV=@n13
     U Frequency=@n14.R
     V Frequency=@n14.R
     U Amplitude=@n14.R
     V Amplitude=@n14.R
     U Speed=@n15
     V Speed=@n15
     U Time=@n16
     V Time=@n16
    ) {materialFunction=/Game/Effects/Fort_Effects/MaterialFunctions/MF_WaveUV}.Wave UVs) {name=UVs, nodeColor=<rgba 1 0 0.4922 1>}
   Low=@n17
   Medium=@n17
  )
  n19 = VectorParameter() {parameterName=Noise Layer 1 Scale and Panner Speed, defaultValue=<rgba 0.8 1 0 1.62>}
  n20 = Multiply(A=@n18, B=ComponentMask(None=@n19.RGB) {g=true, r=true})
  n21 = NamedRerouteUsage()
  n22 = AppendVector(A=@n19.B, B=@n19.B)
  n23 = TextureObject() {
   samplerType=SAMPLERTYPE_LinearGrayscale
   texture=/BRCosmetics/Athena/Effects/Particles/Skydiving/NS_Contrail_DarkAndLight/T_DarkAndLight_Noise
  }
  n24 = VectorParameter() {parameterName=Noise Layer 2 Scale and Panner Speed, defaultValue=<rgba 0.4 0.5 0 1>}
  n25 = Multiply(A=@n18, B=ComponentMask(None=@n24.RGB) {g=true, r=true})
  n26 = AppendVector(A=@n24.B, B=@n24.B)
  n27 = LinearInterpolate(
   A=TextureSample(UVs=Panner(Coordinate=@n20, Time=@n21, Speed=@n22), Tex=@n23).R
   B=TextureSample(UVs=Panner(Coordinate=@n25, Time=@n21, Speed=@n26), Tex=@n23).R
   Alpha=ComponentMask(None=TextureCoordinate()) {g=true}
  )
  n28 = LinearInterpolate(
   A=ScalarParameter() {defaultValue=11, parameterName=Depth Fade Distance Min}
   B=ScalarParameter() {defaultValue=3, parameterName=Depth Fade Distance Max}
   Alpha=@n27
  )
  n29 = ScalarParameter() {defaultValue=0.666, parameterName=Erosion Exponent}
  n30 = MaterialFunctionCall!(
   Gradient=QualitySwitch(Default=DepthFade(Opacity=@n9, FadeDistance=@n28), Low=@n9, Medium=@n9)
   Erosion Amount=OneMinus(None=Power(Base=@n27, Exp=@n29))
  ) {materialFunction=/BRCosmetics/Athena/Effects/Particles/Skydiving/NS_Contrail_DarkAndLight/Functions/MF_ErosionNormalized}
  n31 = TextureSample(UVs=@n30.Result, Tex=TextureObjectParameter(UVs=@n30.Result) {parameterName=FireGradient})
  MaterialFunctionCall(Emissive=Multiply(A=@n1.RGB, B=@n31.RGB)) {materialFunction=&MF_FX_EmissiveAdjustments}
  n32 = DotProduct(A=VertexNormalWS(), B=CameraVectorWS())
  Multiply(A=Multiply(A=@n1.A, B=@n31.A), B=Power(Base=Abs(None=@n32), Exp=ScalarParameter() {defaultValue=2, parameterName=Fresnel Exponent}))
  n33 = Add(A=Time(), B=Multiply(A=ParticleRandom(), B=ScalarParameter() {defaultValue=12.345, parameterName=Randomness Scale}))
  n34 = ScalarParameter() {parameterName=Time Offset}
  NamedRerouteDeclaration(None=Multiply(A=Add(A=@n33, B=@n34), B=ScalarParameter() {defaultValue=0.666, parameterName=Time Scale})) {name=Scaled Time, nodeColor=<rgba 0 0.9766 1 1>}
 ]
}
