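# Unlit, two-sided lightning material for Niagara mesh particles in the CritterBran character's effects.
# A vertex offset widens the bolt, either camera-facing across UVs or along normals by static switch, squeezing tips and edge-on views.
# Dynamic-parameter gradient masks along randomly flipped V drive reveal, thickness and emissive boost of particle color.
# Opacity combines reveal, tip and depth fades, particle alpha and, in spline mode, a core-plus-glow profile.
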Material /BRCosmetics/Characters/Player/Male/Medium/Bodies/M_MED_CritterBran/FX/Materials/M_Character_CritterBran_Lightning
graph {
 properties {bUsedWithNiagaraMeshParticles=true, pixelDepthOffsetMode=PDOM_Legacy, shadingModel=MSM_Unlit, twoSided=true, usedShadingModels=Unlit}
 nodes [
  n1 = Normalize(VectorInput=Subtract(A=CameraPositionWS(), B=WorldPosition().XYZ))
  n2 = Normalize(VectorInput=Transform(None=Constant3Vector() {constant=<rgba 0 1 0 0>}))
  n3 = Normalize(VectorInput=CrossProduct(A=@n1, B=@n2))
  n4 = TextureCoordinate()
  n5 = Multiply(A=Subtract(A=ComponentMask(None=@n4) {r=true}) {constB=0.5}, B=ScalarParameter() {defaultValue=9, parameterName=Lightning Width})
  n6 = Multiply(A=VertexNormalWS(), B=ScalarParameter() {defaultValue=0.666, parameterName=Normal Width})
  n7 = NamedRerouteUsage()
  n8 = ComponentMask(None=@n4) {g=true}
  n9 = OneMinus(None=Abs(None=ConstantBiasScale(None=@n8) {bias=-0.5, scale=2}))
  n10 = ScalarParameter() {defaultValue=0.5, parameterName=Tip Squeeze Exponent}
  n11 = Multiply(A=OneMinus(None=Power(Base=@n9, Exp=@n10)), B=Saturate(None=ScalarParameter() {defaultValue=1, parameterName=Tip Squeeze Amount}))
  n12 = OneMinus(None=Abs(None=DotProduct(A=@n2, B=@n1)))
  n13 = ScalarParameter() {defaultValue=0.5, parameterName=Width View Angle Squish Exponent}
  n14 = Multiply(
   A=OneMinus(None=Power(Base=@n12, Exp=@n13))
   B=Saturate(None=ScalarParameter() {defaultValue=1, parameterName=Width View Angle Squish Amount})
  )
  n15 = Multiply(
   A=StaticSwitchParameter(True=Multiply(A=@n3, B=@n5), False=Multiply(A=@n6, B=@n7) {constA=1}) {parameterName=Use Spline Thicken Tech}
   B=Multiply(A=OneMinus(None=@n11), B=OneMinus(None=@n14))
  )
  n16 = Subtract(A=CameraPositionWS(), B=Add(A=@n15, B=WorldPosition().XYZ))
  Add(A=@n15, B=Multiply(A=Normalize(VectorInput=@n16), B=ScalarParameter() {parameterName=Camera Offset}))
  n17 = ParticleColor()
  n18 = NamedRerouteUsage()
  MaterialFunctionCall(Emissive=LinearInterpolate(A=@n17.RGB, B=Multiply(A=@n17.RGB, B=@n18), Alpha=NamedRerouteUsage())) {materialFunction=&MF_FX_EmissiveAdjustments}
  n19 = Multiply(A=NamedRerouteUsage(), B=NamedRerouteUsage())
  n20 = ComponentMask(None=TextureCoordinate()) {r=true}
  n21 = OneMinus(None=Abs(None=ConstantBiasScale(None=@n20) {bias=-0.5, scale=2}))
  n22 = OneMinus(None=Saturate(None=ScalarParameter() {defaultValue=1, parameterName=Lightning Thickness}))
  n23 = Saturate(None=Divide(A=Subtract(A=@n21, B=@n22), B=ScalarParameter() {parameterName=Lightning Edge Softness}))
  n24 = Power(Base=@n21, Exp=ScalarParameter() {defaultValue=2, parameterName=Lightning Glow Exponent})
  n25 = NamedRerouteUsage()
  n26 = StaticSwitchParameter(True=Multiply(A=Max(A=@n23, B=@n24), B=@n25), False=@n25) {parameterName=Use Spline Thicken Tech}
  Multiply(A=Multiply(A=Multiply(A=@n19, B=@n26), B=@n7), B=ParticleColor().A)
  n27 = ComponentMask(None=TextureCoordinate()) {g=true}
  n28 = LinearInterpolate(A=@n27, B=OneMinus(None=@n27), Alpha=Round(None=ParticleRandom()))
  n29 = DynamicParameter() {paramNames [Lightning Reveal Position, Lightning Reveal Size, Lightning Brightness Boost Size, Lightning Brightness Boost Intensity]}
  n30 = MaterialFunctionCall!(
   Gradient=OneMinus(None=NamedRerouteDeclaration(None=@n28) {name=UV G Randomly Flipped, nodeColor=<rgba 0.2187 1 0 1>})
   Mask Position=@n29.Lightning Reveal Position
   Mask Size=@n29.Lightning Reveal Position
  ) {materialFunction=/Game/Effects/Fort_Effects/MaterialFunctions/MF_GradientMask}
  NamedRerouteDeclaration(None=@n30.Binary) {name=Lightning Opacity Mask, nodeColor=<rgba 0.1953 0 1 1>}
  NamedRerouteDeclaration(None=@n30.Sinusoidal) {name=Lightning Thickness Multiplier, nodeColor=<rgba 1 0.1875 0 1>}
  n31 = ConstantBiasScale(None=ComponentMask(None=TextureCoordinate()) {g=true}) {bias=-0.5, scale=2}
  n32 = Subtract(A=OneMinus(None=Abs(None=@n31)), B=ScalarParameter() {defaultValue=0.1, parameterName=Tip Opacity Fade Offset})
  n33 = ScalarParameter() {defaultValue=0.1, parameterName=Tip Opacity Fade Size}
  NamedRerouteDeclaration(None=Saturate(None=Divide(A=@n32, B=@n33))) {name=Tip Fade, nodeColor=<rgba 1 0 0.4922 1>}
  n34 = DepthFade(FadeDistance=ScalarParameter() {defaultValue=2, parameterName=Depth Fade Distance})
  n35 = Constant() {r=1}
  NamedRerouteDeclaration(None=ShadingPathSwitch(Default=QualitySwitch(Default=@n34, Low=@n35, Medium=@n35), Mobile=@n35) {desc=Migrated from FeatureLevelSwitch}) {name=Depth Fade, nodeColor=<rgba 1 0 0.1172 1>}
  NamedRerouteDeclaration(None=Power(Base=MaterialFunctionCall(Gradient=@n30.Linear, Mask Position=@n29.Lightning Reveal Position, Mask Size=@n29.Lightning Reveal Position) {materialFunction=&MF_GradientMask}.Sinusoidal) {constExponent=6}) {name=Lightning Brightness Boost Mask, nodeColor=<rgba 0.125 0 1 1>}
  NamedRerouteDeclaration(None=@n29.Lightning Brightness Boost Intensity) {name=Lightning Brightness Boost Intensity, nodeColor=<rgba 0.9922 0 1 1>}
  NamedRerouteUsage()
 ]
}
