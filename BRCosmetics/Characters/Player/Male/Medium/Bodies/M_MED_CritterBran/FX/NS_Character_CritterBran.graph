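# Niagara system for the CritterBran character cosmetic, adding sprite eye flames with soft glow and short-lived purple lightning meshes on the body.
# Eyes bursts two particles that follow a socket with mirrored offsets, scale with big-head mode and set visibility from the head-hidden flag.
# SplineThickenMeshes probabilistically spawns bolts with glows, using User int arrays indexed by particle ID to pick mesh, socket and mirroring.
# Bolts on hidden hand or foot sockets are killed.
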NiagaraSystem /BRCosmetics/Characters/Player/Male/Medium/Bodies/M_MED_CritterBran/FX/NS_Character_CritterBran
dependencies {
 usedDataInterfaces [
  &NiagaraDataInterfaceColorCurve
  &NiagaraDataInterfaceArrayFloat3
  &NiagaraDataInterfaceCurve
  &NiagaraDataInterfaceSocketReader
 ]
 usedDynamicInputs []
 usedModules [
  &self:SystemSpawnScript.SystemScriptSource.SystemScriptGraph.NiagaraNodeAssignment_1.SetVariables_#1
  &SystemState
 ]
}
emitters {
 Eyes {
  emitterSpawnScript [/SetVariables(Emitter.FireGradientEye=DataInterface() {propertyValues {
      ExposedName=NiagaraFireEyeGradient
      bExposeCurve=true
      AlphaCurve {keys [<at 0 0>,<at 0.432 0.2>,<at 0.8872 0.7>,<at 1 0>]}
      BlueCurve {keys [<at 0 0.0067>,<at 0.573 0.5>,<at 0.7253 1.5>,<at 1 1.5>]}
      GreenCurve {keys [<at 0 0>,<at 0.573 0>,<at 0.7253 0>,<at 1 0>]}
      RedCurve {keys [<at 0 0.01>,<at 0.573 0.4167>,<at 0.7253 0.5>,<at 1 1.25>]}
     }}) {isSetParameters=true}]
  emitterUpdateScript [
   &EmitterState(
    Inactive Response=::NewEnumerator1
    Life Cycle Mode=::NewEnumerator1
    Loop Behavior=::NewEnumerator1
    Loop Duration Mode=::NewEnumerator1
    UseLoopDelay=false
   )
   &SpawnBurst_Instantaneous(Spawn Count=2, Use Spawn Probability=false, UseLoopCountLimit=false)
  ]
  particleSpawnScript [
   &InitializeParticle(Color=<rgba 0.6667 0 2 1>, UsePositionOffset=false, Write Material Random=true)
   &PartitionParticles()
   /SetVariables(Particles.SoftGlowSize:Vector2f=&Vector2DFromFloat(Value=16)) {isSetParameters=true}
   /SetVariables(Particles.FlameCameraOffset=5, Particles.FlameSize:Vector2f=<11 7>) {isSetParameters=true}
  ]
  particleUpdateScript [
   &SocketLocation(
    Orientation Sampling=::NewEnumerator1
    SamplePosOffset=true
    Sampled Position Offset=&Multiply_VectorByFloat(
     Vector=&SelectVectorFromArray(
      Direct Array Index=$Particles.PartitionParticles.Partition
      Direct Array Mode=::NewEnumerator1
      Vector Selection Array=DataInterface() {propertyValues {FloatData [<3 9 3>, <3 9 -3>]}}
     )
     Float=&Lerp_Float(A=1, B=2, Alpha=$NPC.FortniteNPC.BigHeadMode)
    )
    Socket Index=2
    Socket Reader=$System.Socket Reader
    Socket Sampling Mode=::NewEnumerator3
    VelocitySpeedLimit=false
   )
   &SpriteFacingAndAlignment(
    Sprite Alignment=&RotateVectorByQuaternion(
     Vector To Rotate=&SelectVectorFromArray(
      Direct Array Index=$Particles.PartitionParticles.Partition
      Direct Array Mode=::NewEnumerator1
      Vector Selection Array=DataInterface() {propertyValues {FloatData [<1 1 0.5>, <1 1 -0.5>]}}
     )
     Quaternion=$Output.SocketLocation.SampledRotation
    )
    Sprite Facing=&RotateVectorByQuaternion(
     Vector To Rotate=&SelectVectorFromArray(
      Direct Array Index=$Particles.PartitionParticles.Partition
      Direct Array Mode=::NewEnumerator1
      Vector Selection Array=DataInterface() {propertyValues {FloatData [<0 1 0.5>, <0 1 -0.5>]}}
     )
     Quaternion=$Output.SocketLocation.SampledRotation
    )
    WriteAlignment=true
    WriteFacing=true
   )
   /SetVariables(Particles.VisibilityTag=&MakeCustomIntFromBool(Bool=$User.bisHeadHidden, True INT=1, False INT=0)) {isSetParameters=true}
   /SetVariables(
    Particles.FlameSize:Vector2f=&Multiply_Vector2D_ByFloat(Float=&Lerp_Float(A=1, B=1.8, Alpha=$NPC.FortniteNPC.BigHeadMode), Vector2D=$Particles.Initial.FlameSize)
    Particles.SoftGlowSize:Vector2f=&Multiply_Vector2D_ByFloat(Float=&Lerp_Float(A=1, B=1.8, Alpha=$NPC.FortniteNPC.BigHeadMode), Vector2D=$Particles.Initial.SoftGlowSize)
   ) {isSetParameters=true}
  ]
  rendererClasses [&NiagaraSpriteRendererProperties]
  renderers [
   {rendererClass=&NiagaraSpriteRendererProperties}
   {rendererClass=&NiagaraSpriteRendererProperties, rendererIndex=1}
  ]
 }
 SplineThickenMeshes {
  emitterUpdateScript [
   &EmitterState(
    Inactive Response=::NewEnumerator1
    Life Cycle Mode=::NewEnumerator1
    Loop Behavior=::NewEnumerator1
    Loop Duration Mode=::NewEnumerator1
    UseLoopDelay=false
   )
   &SpawnRate(Spawn Probability=0.62, SpawnRate=6, Use Spawn Probability=true)
  ]
  particleSpawnScript [
   /SetVariables(
    Particles.MeshIndex=&Add_Integer(
     A=&SelectIntFromArray(
      Direct Array Index=&ReturnParticleIDAttributes()
      Direct Array Mode=::NewEnumerator1
      Int Selection Array=$User.LightningMeshIndex
     )
     B=&Multiply_Int(
      A=&SelectIntFromArray(
       Direct Array Index=&ReturnParticleIDAttributes()
       Direct Array Mode=::NewEnumerator1
       Int Selection Array=$User.LightningSocketIndex
      )
      B=3
     )
    )
    Particles.SocketIndex:NiagaraInt32=&Add_Integer(
     A=&SelectIntFromArray(
      Direct Array Index=&SelectIntFromArray(
       Direct Array Index=&ReturnParticleIDAttributes()
       Direct Array Mode=::NewEnumerator1
       Int Selection Array=$User.LightningSocketIndex
      )
      Direct Array Mode=::NewEnumerator1
      Int Selection Array=DataInterface() {propertyValues {IntData [3,5,7]}}
     )
     B=&SelectIntFromArray(
      Direct Array Index=&ReturnParticleIDAttributes()
      Direct Array Mode=::NewEnumerator1
      Int Selection Array=$User.LightningMirroring
     )
    )
   ) {isSetParameters=true}
   /SetVariables(Particles.MeshScaleFactor:Vector3f=&MakeCustomVectorFromBool(
     Bool=&SetBoolByIntComparison(
      A=&SelectIntFromArray(
       Direct Array Index=&ReturnParticleIDAttributes()
       Direct Array Mode=::NewEnumerator1
       Int Selection Array=$User.LightningMirroring
      )
      B=0
      Comparison Type=::NewEnumerator5
     )
     True Vector=&MakeCustomVectorFromBool(
      Bool=&SetBoolByIntComparison(A=$Particles.MeshIndex, B=2, Comparison Type=::NewEnumerator7)
      True Vector=<-1 1 -1>
      False Vector=<-1 -1 -1>
     )
     False Vector=<1 1 1>
    )) {isSetParameters=true}
   &InitializeParticle(
    Color Mode=::NewEnumerator0
    Lifetime=&RandomRangeFloat(Minimum=0.3, Maximum=0.46)
    Mesh Scale=$Particles.MeshScaleFactor
    Mesh Scale Mode=::NewEnumerator4
    Sprite Size Mode=::NewEnumerator3
    Uniform Sprite Size=44
    UsePositionOffset=false
    Write Material Random=true
   )
   /SetVariables(Particles.GlowSpriteCameraOffset=22) {isSetParameters=true}
   &KillParticles(Kill Particles=&Boolean_OrOperation(
     Bool1=&Boolean_AndOperation(
      Bool1=$User.bisHandHidden
      Bool2=&Boolean_OrOperation(
       Bool1=&SetBoolByIntComparison(A=$StackContext.SocketIndex, B=3, Comparison Type=::NewEnumerator4)
       Bool2=&SetBoolByIntComparison(A=$StackContext.SocketIndex, B=4, Comparison Type=::NewEnumerator4)
      )
     )
     Bool2=&Boolean_AndOperation(
      Bool1=$User.bisFootHidden
      Bool2=&Boolean_OrOperation(
       Bool1=&SetBoolByIntComparison(A=$StackContext.SocketIndex, B=7, Comparison Type=::NewEnumerator4)
       Bool2=&SetBoolByIntComparison(A=$StackContext.SocketIndex, B=8, Comparison Type=::NewEnumerator4)
      )
     )
    ))
  ]
  particleUpdateScript [
   &ParticleState()
   &SocketLocation(
    SamplePosOffset=false
    Socket Index=$Particles.SocketIndex
    Socket Reader=$System.Socket Reader
    Socket Sampling Mode=::NewEnumerator3
    VelocitySpeedLimit=false
   )
   &Color(Color=&ColorFromCurve(ColorCurve=DataInterface() {propertyValues {
       AlphaCurve {keys [<at 0 1 RCIM_Cubic>,<at 1 1 RCIM_Cubic>]}
       BlueCurve {keys [<at 0 11 RCIM_Cubic>,<at 0.5925 1 RCIM_Cubic>]}
       GreenCurve {keys [<at 0 0 RCIM_Cubic>,<at 0.5925 0 RCIM_Cubic>]}
       RedCurve {keys [<at 0 9.1667 RCIM_Cubic>,<at 0.5925 0.1667 RCIM_Cubic>]}
      }}))
   &DynamicMaterialParameters(
    Index 0 Param 1=&FloatFromCurve(FloatCurve=DataInterface() {propertyValues {Curve {keys [{tangentMode=RCTM_User, time=0, value=0}, <at 0.333 0.5>, {tangentMode=RCTM_User, time=1, value=1}]}}})
    Index 0 Param 2=&FloatFromCurve(FloatCurve=DataInterface() {propertyValues {Curve {keys [<at 0.333 1>,<at 0.666 0.333>]}}})
    Index 0 Param 3=0.75
    Index 0 Param 4=&FloatFromCurve(
     FloatCurve=DataInterface() {propertyValues {Curve {keys [
         {interpMode=RCIM_Cubic, tangentMode=RCTM_User, time=0, value=0}
         {interpMode=RCIM_Cubic, tangentMode=RCTM_User, time=0.5, value=1}
         {tangentMode=RCTM_User, time=1, value=0}
        ]}}}
     Scale Curve=33
    )
    Module.Index0Param0Write=true
    Module.Index0Param1Write=true
    Module.Index0Param2Write=true
    Module.Index0Param3Write=true
   )
   /SetVariables(
    Particles.GlowColor=&MultiplyLinearColor(Linear Color=$Particles.Color, Scale Factor=<1 1 1 0.1>)
    Particles.GlowSpriteColor:LinearColor=&MultiplyLinearColor(Linear Color=$Particles.Color, Scale Factor=<1 1 1 0.02>)
    Particles.GlowSpriteDynamic:Vector4f=&MakeVector4(
     X=&Lerp_Float(
      A=6
      B=2
      Alpha=&FloatFromCurve(FloatCurve=DataInterface() {propertyValues {Curve {keys [
           {arriveTangent=9.0725, interpMode=RCIM_Cubic, leaveTangent=9.0725, tangentMode=RCTM_User, time=0, value=0}
           {arriveTangent=-2.5249, interpMode=RCIM_Cubic, leaveTangent=-2.5249, tangentMode=RCTM_User, time=0.1408, value=0.9847}
           {arriveTangent=-0.0405, interpMode=RCIM_Cubic, leaveTangent=-0.0405, tangentMode=RCTM_User, time=1, value=0}
          ]}}})
     )
     Y=0
     Z=0
     W=0
    )
   ) {isSetParameters=true}
  ]
  rendererClasses [&NiagaraMeshRendererProperties, &NiagaraSpriteRendererProperties]
  renderers [
   {rendererClass=&NiagaraMeshRendererProperties, rendererIndex=0}
   {rendererClass=&NiagaraMeshRendererProperties, rendererIndex=1}
   {rendererClass=&NiagaraSpriteRendererProperties, rendererIndex=2}
  ]
 }
}
summary {
 systemName=NS_Character_CritterBran
 emitters [
  {emitterName=Eyes, rendererClasses [&NiagaraSpriteRendererProperties]}
  {
   emitterName=SplineThickenMeshes
   rendererClasses [&NiagaraMeshRendererProperties, &NiagaraSpriteRendererProperties]
  }
 ]
 userVariables [
  User.bisFootHidden {type=NiagaraBool}
  User.bIsFrontEnd {type=NiagaraBool}
  User.bIsFrontEndPreview {type=NiagaraBool}
  User.bisHandHidden {type=NiagaraBool}
  User.bisHeadHidden {type=NiagaraBool}
  User.LightningMeshIndex {
   type=NiagaraDataInterfaceArrayInt32
   description=""
   defaultValue {
    struct=NiagaraExt_VariableValue_DataInterface
    value {
     dataInterface=&self:NiagaraDataInterfaceArrayInt32_3
     dataInterfaceClass=&NiagaraDataInterfaceArrayInt32
    }
   }
  }
  User.LightningMirroring {
   type=NiagaraDataInterfaceArrayInt32
   description=""
   defaultValue {
    struct=NiagaraExt_VariableValue_DataInterface
    value {
     dataInterface=&self:NiagaraDataInterfaceArrayInt32_1
     dataInterfaceClass=&NiagaraDataInterfaceArrayInt32
    }
   }
  }
  User.LightningSocketIndex {
   type=NiagaraDataInterfaceArrayInt32
   description=""
   defaultValue {
    struct=NiagaraExt_VariableValue_DataInterface
    value {
     dataInterface=&self:NiagaraDataInterfaceArrayInt32_2
     dataInterfaceClass=&NiagaraDataInterfaceArrayInt32
    }
   }
  }
 ]
}
